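# Comet Summer backpack material on the Fortnite character FX stack, with switchable neon mouth and eye shapes and an optional anime shading switch.
# Sphere masks compare a mesh mask against either a time-driven eye blink or reactive shape values.
# Reactive timing derives from a Blueprint-supplied game time and fade length.
# World Position Offset collapses unselected shapes to the actor position, hiding them without opacity.
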Material /Game/Accessories/FORT_Backpacks/M_MED_Comet_Summer/Materials/M_Med_CometSummer_BackPack_Custom
graph {
 properties {
  bUseMaterialAttributes=true
  bUsedWithClothing=true
  bUsedWithMeshDeformer=true
  bUsedWithMorphTargets=true
  bUsedWithSkeletalMesh=true
  pixelDepthOffsetMode=PDOM_Legacy
  refractionMethod=RM_IndexOfRefraction
  usedShadingModels=Default Lit
 }
 nodes [
  n1 = MaterialFunctionCall() {materialFunction=&BaseFN_CharacterFunction_PreFX}
  n2 = MaterialFunctionCall(
   PreskinnedPosition=@n1.Pre skinned Local Position
   Gmap=@n1.Pre skinned Local Position
   MaterialAttributes=@n1.Pre skinned Local Position
   AO=@n1.Pre skinned Local Position
  ) {materialFunction=&CharacterSkinFX}
  n3 = ActorPositionWS()
  n4 = WorldPosition()
  n5 = LinearInterpolate(
   A=ScalarParameter() {group=Neon, parameterName=Neon_ReactiveShape_Mouth}
   B=ScalarParameter() {group=Neon, parameterName=Neon_ReactiveShape_Eyes}
   Alpha=NamedRerouteUsage()
  )
  n6 = Time()
  n7 = ScalarParameter() {defaultValue=0.2, group=Neon, parameterName=BlinkFrequency}
  n8 = Multiply(
   A=Frac(None=Multiply(A=@n6, B=@n7) {constB=0.2, desc=frequency of blink})
   B=ScalarParameter() {defaultValue=1.04, group=Neon, parameterName=BlinkLength}
  ) {constB=1.04, desc=">1 = length of blink"}
  n9 = Multiply(
   A=NamedRerouteUsage()
   B=Multiply(A=Floor(None=@n8), B=ScalarParameter() {defaultValue=0.25, group=Neon, parameterName=BlinkValue}) {constB=0.25}
  ) {desc=Mask to Eyes}
  n10 = NamedRerouteUsage()
  n11 = Saturate(None=SphereMask(
    A=NamedRerouteUsage()
    B=LinearInterpolate(A=@n5, B=@n9, Alpha=@n10) {constB=0, desc=lerp between reactive A and passive B}
    Radius=Constant() {r=0.13}
    Hardness=Constant() {r=1}
   ))
  n12 = SetMaterialAttributes!(
   MaterialAttributes=StaticSwitchParameter(
    True=MaterialFunctionCall!(
     Material=@n2.Result
     LitDiffuseOverride
     SSC_AlphaOverride
     Tint Mask
     DF InkLine Visibility Mask
     DF Inkline Injection
     DF Inkline Color Override
     DF Inkline Color Override Mask
     CustomUVs
    ) {materialFunction=/Game/Packages/Fortress_SharedMaterials/Anime_Master_Material/MF_AnimeShading}.Result
    False=GetMaterialAttributes(None=@n2.Result) {attributeGetTypes [#1, #2]}.MaterialAttributes
   ) {group=Anime Shading, parameterName=Use Anime Shading}
   Base Color
   Metallic
   Specular
   Roughness
   Emissive Color
   Normal
   World Position Offset=LinearInterpolate(A=Subtract(A=@n3, B=@n4.XYZ), Alpha=Round(None=@n11)) {constA=1, constB=0, desc="deactivated = 0"}
  ) {attributeSetTypes [
    #1
    #3
    #4
    #5
    #6
    #7
    #2
   ]}
  n13 = MaterialFunctionCall(AO=@n1.AO, Distance Blend=@n1.AO, Fresnel=@n1.AO, Gmap=@n1.AO, Pre Skinned Local Position=@n1.AO, Material=@n12) {materialFunction=&BaseFN_CharacterFunction_PostFX}
  MaterialFunctionCall(In=@n13.Result) {materialFunction=&CINE_MF_CharPostFX}
  n14 = VertexColor()
  NamedRerouteDeclaration(None=@n14.R) {name=VertexColorRed, nodeColor=<rgba 1 0 0 1>}
  NamedRerouteDeclaration(None=@n14.G) {name=VertexColorGreen, nodeColor=<rgba 0 1 0 1>}
  NamedRerouteDeclaration(None=@n14.B) {name=VertexColorBlue, nodeColor=<rgba 0 0 1 1>}
  n15 = Time()
  n16 = ScalarParameter() {group=Neon, parameterName=Neon_GameTimeFromBP}
  n17 = Divide(
   A=Max(A=Subtract(A=@n15, B=@n16)) {constB=0}
   B=Multiply(A=ScalarParameter() {defaultValue=1, group=Neon, parameterName=Neon_ReactiveLength}) {constB=1.5, desc=shapes stay on longer}
  ) {desc=the length of time the reaction takes to fade}
  NamedRerouteDeclaration(None=Floor(None=Saturate(None=@n17))) {name=ReactivityLerpAlpha, nodeColor=<rgba 0 1 0.25 1>}
 ]
}
